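# Level Sequence for tac5100 in PistolCanary, apparently a shot-level container that splits content by department through subsequence tracks.
# It has separate audio, camera, gameplay, lighting, techart, vfx and animation tracks, with no object bindings of its own.
# Only the animation track holds a section, spanning frames 3600 to 3818, and the other tracks are empty placeholders.
# Playback displays at 60 fps.
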LevelSequence /PistolCanary/Sequence/tac/tac5100/PistolCanary_tac5100
bindings []
display_rate {denominator=1, numerator=60}
tick_resolution {denominator=1, numerator=24000}
tracks [
 {type=MovieSceneSubTrack, display_name=audio, sections []}
 {type=MovieSceneSubTrack, display_name=camera, sections []}
 {type=MovieSceneSubTrack, display_name=gameplay, sections []}
 {type=MovieSceneSubTrack, display_name=lighting, sections []}
 {type=MovieSceneSubTrack, display_name=techart, sections []}
 {type=MovieSceneSubTrack, display_name=vfx, sections []}
 {type=MovieSceneSubTrack, display_name=animation, sections [{type=MovieSceneSubSection, end_frame=3818, start_frame=3600}]}
]
